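; BP_MossSpawner is a placeable tool actor that refreshes procedural moss around its location through PCG.
; Its construction script forces uniform scale, tracks movement, calls Update Moss and draws a scale-sized debug sphere.
; Update Moss regenerates enclosing PCG volumes tagged "updatePCG", plus BP_ProceduralMossTile actors whose margin-padded bounds contain it.
; A separate Update UniqueMossMeshes event does the same for "UniqueMoss"-tagged volumes only.
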(event EventBeginPlay)

(event Collision|EventActorBeginOverlap
  (params (OtherActor :type Actor%20Object%20Reference)))

(event EventTick
  (params (DeltaSeconds :type Float%20%28single-precision%29)))

(event Custom|Update%20Moss
  (Utilities|FlowControl|Sequence
    (:then_0
      (bind _outactors (Actor|GetAllActorsOfClass "/Script/PCG.PCGVolume"))
      (for _array_element _outactors
        (bind _returnvalue (PCG|Helpers|GetActorBoundsPCG _array_element))
        (if (GeometryScript|Shapes|Box|TestPointInsideBox _returnvalue (Transformation|GetActorLocation))
          (for _array_element_2 (Variables|/Script/Engine.Actor|GetTags _array_element)
            (if (== _array_element_2 "updatePCG")
              (PCG|Cleanup (Variables|/Script/PCG.PCGVolume|GetPCGComponent _array_element) false)
              (PCG|Generate (Variables|/Script/PCG.PCGVolume|GetPCGComponent _array_element) false))))))
    (:then_1
      (bind _outactors_2 (Actor|GetAllActorsOfClass "/Game/ArtTools/Procedural/PCGMoss/BP_ProceduralMossTile.BP_ProceduralMossTile_C"))
      (for _array_element_3 _outactors_2
        (bind _returnvalue_2 (PCG|Helpers|GetActorBoundsPCG _array_element_3))
        (if (GeometryScript|Shapes|Box|TestPointInsideBox (Math|Box|MakeBox (- (Utilities|Struct|BreakBox _returnvalue_2) (Variables|Default|GetMargin)) (+ (.Max (Utilities|Struct|BreakBox _returnvalue_2)) (Variables|Default|GetMargin))) (Transformation|GetActorLocation))
          (PCG|CleanupLocal (Variables|/Game/ArtTools/Procedural/PCGMoss/BP_ProceduralMossTile.BP_ProceduralMossTile_C|GetPCG _array_element_3) false)
          (PCG|Generate (Variables|/Game/ArtTools/Procedural/PCGMoss/BP_ProceduralMossTile.BP_ProceduralMossTile_C|GetPCG _array_element_3) false))))))

(event Custom|Update%20UniqueMossMeshes
  (Utilities|FlowControl|Sequence
    (:then_0
      (bind _outactors (Actor|GetAllActorsOfClass "/Script/PCG.PCGVolume"))
      (for _array_element _outactors
        (bind _returnvalue (PCG|Helpers|GetActorBoundsPCG _array_element))
        (if (GeometryScript|Shapes|Box|TestPointInsideBox _returnvalue (Transformation|GetActorLocation))
          (for _array_element_2 (Variables|/Script/Engine.Actor|GetTags _array_element)
            (if (== _array_element_2 "UniqueMoss")
              (PCG|Cleanup (Variables|/Script/PCG.PCGVolume|GetPCGComponent _array_element) false)
              (PCG|Generate (Variables|/Script/PCG.PCGVolume|GetPCGComponent _array_element) false))))))))

(fn ConstructionScript
  (bind _returnvalue (Transformation|GetActorLocation))
  (bind _returnvalue_2 (Transformation|GetActorScale3D))
  (bind _x (.x _returnvalue_2))
  (Utilities|FlowControl|Sequence
    (:then_0
      (Transformation|SetActorScale3D (Math|Vector|MakeVector _x _x _x))
      (if (Variables|Default|GetLiveEdit)
        _
        (elif (Utilities|Operators|Equal(==) (Variables|Default|GetStoredLocation) _returnvalue 0.1)
          _
          (else
            (Variables|Default|SetLiveEdit true))))
      (Variables|Default|SetStoredLocation _returnvalue)
      (Utilities|FlowControl|Sequence
        (:then_0
          (CallFunction|UpdateMoss))
        (:then_1
          (Rendering|Debug|DrawDebugSphere (Transformation|GetActorLocation) (* _x 5e+02) 12 "(R=1.000000,G=0.098177,B=0.000000,A=1.000000)" 1.5 1.0))))))
