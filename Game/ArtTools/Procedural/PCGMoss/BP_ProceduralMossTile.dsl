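; Procedural moss tile actor, apparently for PCG, that sets itself up on a streaming-aligned grid in its Construction Script.
; It snaps to the cell centre at a fixed height and sizes its box from grid size, tiles per streaming grid and volume height.
; It compares locations with other tiles, printing duplicates and swapping billboard sprites as an editor warning.
; Its instanced static meshes use contact shadows instead of regular shadows.
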(event EventBeginPlay)

(event Collision|EventActorBeginOverlap
  (params (OtherActor :type Actor%20Object%20Reference)))

(event EventTick
  (params (DeltaSeconds :type Float%20%28single-precision%29)))

(fn ConstructionScript
  (Utilities|FlowControl|Sequence
    (:then_0
      (Transformation|SetActorLocation (Math|Vector|MakeVector (* (- (Math|Float|Ceil (/ (.x (Transformation|GetActorLocation)) (/ (Variables|Default|GetGridSize) (Variables|Default|GetTilesPerStreamingGrid)))) 0.5) (/ (Variables|Default|GetGridSize) (Variables|Default|GetTilesPerStreamingGrid))) (* (- (Math|Float|Ceil (/ (.y (Transformation|GetActorLocation)) (/ (Variables|Default|GetGridSize) (Variables|Default|GetTilesPerStreamingGrid)))) 0.5) (/ (Variables|Default|GetGridSize) (Variables|Default|GetTilesPerStreamingGrid))) 5e+03))
      (Components|Box|SetBoxExtent (Variables|Default|GetBox) (Math|Vector|MakeVector (/ (/ (Variables|Default|GetGridSize) (Variables|Default|GetTilesPerStreamingGrid)) 2.0) (/ (/ (Variables|Default|GetGridSize) (Variables|Default|GetTilesPerStreamingGrid)) 2.0) (Variables|Default|GetVolumeHeight)) true)
      (Variables|Rendering|SetSpriteScale 5.0)
      (bind _outactors (Actor|GetAllActorsOfClass "/Game/ArtTools/Procedural/PCGMoss/BP_ProceduralMossTile.BP_ProceduralMossTile_C"))
      (bind _array_element (Utilities|Array|ForEachLoopwithBreak _outactors)
        (:LoopBody
          (if (!= _array_element self)
            (if (Utilities|Operators|Equal(==) (Transformation|GetActorLocation _array_element) (Transformation|GetActorLocation) 5.0)
              (Development|PrintString (Utilities|String|Append (Utilities|String|Append (Utilities|GetDisplayName _array_element) "   AND   ") (Utilities|GetDisplayName self)) true true "(R=1.000000,G=0.057065,B=0.000000,A=1.000000)")
              (Utilities|FlowControl|Sequence
                (:then_0
                  (Rendering|Components|Sprite|SetSprite (Variables|Default|GetBillboard) "/Engine/EditorResources/BadBlueprintSprite.BadBlueprintSprite")
                  (Variables|Rendering|SetSpriteScale 3e+01))
                (:then_1
                  _)))))
        (:Completed
          _))
      (Utilities|FlowControl|Gate false
        (:Exit
          (Rendering|Components|Sprite|SetSprite (Variables|Default|GetBillboard) "/Engine/EditorResources/S_TriggerBox.S_TriggerBox")
          (Variables|Rendering|SetSpriteScale 5.0))))
    (:then_1
      (for _array_element_2 (Actor|GetComponentsByClass "/Script/Engine.InstancedStaticMeshComponent")
        (Rendering|SetCastShadow _array_element_2 false)
        (Rendering|SetCastContactShadow _array_element_2 true)))))
