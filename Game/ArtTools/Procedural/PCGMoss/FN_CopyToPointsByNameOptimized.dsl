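; PCG Blueprint element for procedural moss that copies input points onto every placed static mesh whose asset path matches the point's "Name" attribute.
; It first groups world transforms by mesh path from "Moss"-tagged actors overlapping the PCG owner's bounds.
; A variable loop then composes each point with its mesh's transforms, keeping only results inside those bounds.
; Kept points get density 1 and an index-offset seed.
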(fn Execute
  (params (Input :type PCGData%20Collection%20Structure%20%28by%20ref%29))
  (bind _output (|CallParentFunction Input))
  (return _output))

(fn ExecuteWithContext
  (params (InContext :type PCGContext%20Structure%20%28by%20ref%29) (Input :type PCGData%20Collection%20Structure%20%28by%20ref%29))
  (bind _listnames (Variables|Default|GetListNames))
  (LocalVariables|SetContext InContext)
  (bind _returnvalue (Data|GetAllSpatialInputs Input))
  (Utilities|FlowControl|Sequence
    (:then_0
      (bind _returnvalue_2 (PCG|Temporary|GetOriginalComponent (LocalVariables|GetContext)))
      (bind _returnvalue_3 (PCG|Helpers|GetActorBoundsPCG (Components|GetOwner _returnvalue_2)))
      (Variables|Default|SetBounds _returnvalue_3)
      (bind _outactors (Actor|GetAllActorswithTag "Moss"))
      (for _array_element _outactors
        (bind _returnvalue_4 (PCG|Helpers|GetActorBoundsPCG _array_element))
        (if (Class|BTechArtGlobalLibrary|AreBoundsOverlapping? _returnvalue_4 _returnvalue_3)
          (for _array_element_2 (Actor|GetComponentsByClass _array_element "/Script/Engine.StaticMeshComponent")
            (Utilities|IsValid (Variables|/Script/Engine.StaticMeshComponent|GetStaticMesh _array_element_2)
              (:Is%20Valid
                (if (>= (Utilities|Array|FindItem _listnames (Utilities|GetObjectPathString (Variables|/Script/Engine.StaticMeshComponent|GetStaticMesh _array_element_2))) 0)
                  (Utilities|Array|Add (Utilities|Struct|BreakFNPCGStructMeshSamples (Utilities|Array|Get(aref) (Variables|Default|GetListTransforms) (Utilities|Array|FindItem _listnames (Utilities|GetObjectPathString (Variables|/Script/Engine.StaticMeshComponent|GetStaticMesh _array_element_2))))) (Transformation|GetWorldTransform _array_element_2))
                  (else
                    (Utilities|Array|Add (Variables|Default|GetListNames) (Utilities|GetObjectPathString (Variables|/Script/Engine.StaticMeshComponent|GetStaticMesh _array_element_2)))
                    (Utilities|Array|Add (Variables|Default|GetListTransforms) (Utilities|Struct|MakeFNPCGStructMeshSamples (Utilities|Array|MakeArray (Transformation|GetWorldTransform _array_element_2))))))))))))
    (:then_3
      (bind _aspcgattribute_set (Utilities|Casting|CastToPCGParamData (PCG|TaggedData|BreakTaggedData (Utilities|Array|Get(acopy) _returnvalue))))
      (Variables|Default|Setmetadata (Metadata|ConstMetadata _aspcgattribute_set)))
    (:then_4
      (bind _aspcgspatial_data (Utilities|Casting|CastToPCGSpatialData (PCG|TaggedData|BreakTaggedData (Utilities|Array|Get(acopy) _returnvalue))))
      (bind _returnvalue_5 (Game|ConstructObjectfromClass "/Script/PCG.PCGPointData"))
      (bind _returnvalue_6 (PCG|Temporary|GetComponent (Variables|GetInContext)))
      (Variables|/Script/PCG.PCGSpatialData|SetTargetActor (Components|GetOwner _returnvalue_6) _returnvalue_5)
      (bind _output_get (LocalVariables|SetOutData _returnvalue_5))
      (PCG|FlowControl|VariableLoop (LocalVariables|GetContext) (SpatialData|ToPointDatawithContext _aspcgspatial_data (LocalVariables|GetContext)) _output_get))
    (:then_5
      (return (Utilities|Struct|MakePCGDataCollection (Utilities|Array|MakeArray (PCG|TaggedData|MakeTaggedData (LocalVariables|GetOutData))))))))

(fn VariableLoopBody
  (params (InContext :type PCGContext%20Structure%20%28by%20ref%29) (InData :type PCGPoint%20Data%20Object%20Reference) (InPoint :type PCGPoint%20Structure%20%28by%20ref%29) (OutMetadata :type PCGMetadata%20Object%20Reference) (Iteration :type Integer64))
  (bind (_transform _density _boundsmin _boundsmax _color _steepness _seed) (Utilities|Struct|BreakPCGPoint InPoint))
  (bind _listnames (Variables|Default|GetListNames))
  (bind _returnvalue (PCG|Metadata|GetStringAttribute InPoint (Metadata|ConstMetadata InData) "Name"))
  (if (>= (Utilities|Array|FindItem _listnames _returnvalue) 0)
    (for (_array_element _array_index) (Utilities|Struct|BreakFNPCGStructMeshSamples (Utilities|Array|Get(acopy) (Variables|Default|GetListTransforms) (Utilities|Array|FindItem _listnames _returnvalue)))
      (if (Class|BTechArtGlobalLibrary|AreBoundsOverlapping? (Math|Box|MakeBox (- (.location (Math|Transform|ComposeTransforms _transform _array_element)) "1.000000,1.000000,1.000000") (+ (.location (Math|Transform|ComposeTransforms _transform _array_element)) "1.000000,1.000000,1.000000")) (Variables|Default|GetBounds))
        (Utilities|Array|Add (LocalVariables|GetPointlist) (Utilities|Struct|MakePCGPoint (Math|Transform|ComposeTransforms _transform _array_element) 1.0 "-1.000000,-1.000000,-1.000000" "1.000000,1.000000,1.000000" (+ _array_index _seed)))))
    (return (LocalVariables|GetPointlist))))
