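; PCG graph that spawns instanced moss clumps on tagged meshes, concentrated around MossSpawn-tagged actors.
; Mesh samples are copied by name onto UniqueMoss data from another graph's component, then faded by inverted distance to spawn actors.
; Density drives transforms before self-pruning, and a height comparison against lowered landscape projections culls buried points before spawning.
; A landscape-sampling branch never reaches the spawner, and several older nodes are unconnected.
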(graph FN_JungleMoss
  (inputs (OriginalActor :type Spatial :advanced) (Landscape :type Surface :advanced) (NewPin :type Any))
  (bind Surface%20Sampler_1 (Surface%20Sampler :PointExtents (11.0 11.0 11.0) :PointsPerSquaredMeter 2.0 :PointSteepness 0.5))
  (bind ExecuteBlueprint_1 (Execute%20Blueprint :BlueprintElementInstance (class:/Game/ArtTools/Procedural/PCGMoss/FN_SampleMeshbyTagUnique.FN_SampleMeshbyTagUnique_C) :BlueprintElementType /Game/ArtTools/Procedural/PCGMoss/FN_SampleMeshbyTagUnique.FN_SampleMeshbyTagUnique_C))
  (bind Get%20Actor%20Data_20 (Get%20Actor%20Data :ActorSelector (struct :ActorFilter Root :bIncludeChildren true :ActorSelectionTag UniqueMoss) :bComponentsMustOverlapSelf false :Mode GetDataFromPCGComponent)) ; get unique meshes from other Graph
  (bind ExecuteBlueprint_0 (Execute%20Blueprint :BlueprintElementInstance (class:/Game/ArtTools/Procedural/PCGMoss/FN_CopyToPointsByNameOptimized.FN_CopyToPointsByNameOptimized_C) :BlueprintElementType /Game/ArtTools/Procedural/PCGMoss/FN_CopyToPointsByNameOptimized.FN_CopyToPointsByNameOptimized_C))
  (bind Density%20Filter_4 (Density%20Filter))
  (bind ExecuteBlueprint_11 (Execute%20Blueprint :BlueprintElementInstance (class:/PCG/BP_Elements/Deprecated/PointsFromActorTag.PointsFromActorTag_C :ActorTag MossSpawn) :BlueprintElementType /PCG/BP_Elements/Deprecated/PointsFromActorTag.PointsFromActorTag_C))
  (bind Distance_33 (Distance :bSetDensity true :MaximumDistance 5e+02 :SourceShape Center :TargetShape Center))
  (bind DensityRemap_0 (Density%20Remap :OutRangeMax 0.0 :OutRangeMin 1.0))
  (bind Density%20Filter_35 (Density%20Filter :LowerBound 0.0))
  (bind Transform%20Points_43 (Transform%20Points :AttributeName Density :bAbsoluteOffset true :bAbsoluteScale true :bRecomputeSeed true :bUniformScale false :OffsetMax (0.0 0.0 -25.0) :OffsetMin (0.0 0.0 -25.0) :RotationMax (struct :Yaw 3.6e+02) :ScaleMax (1.0 1.0 1.5)))
  (bind ExtentsModifier_46 (Extents%20Modifier :Extents (9.0 9.0 9.0)))
  (bind ExecuteBlueprint_5 (Execute%20Blueprint :BlueprintElementInstance (class:/Game/ArtTools/Procedural/PCGMoss/FN_TransformByDensity.FN_TransformByDensity_C :Density%20Exp 0.5 :Transform (struct :Translation (0.0 0.0 -5e+01) :Scale3D (1.0 1.0 1.5))) :BlueprintElementType /Game/ArtTools/Procedural/PCGMoss/FN_TransformByDensity.FN_TransformByDensity_C))
  (bind Self%20Pruning_45 (Self%20Pruning))
  (bind ExecuteBlueprint_6 (Execute%20Blueprint :BlueprintElementInstance (class:/PCG/BP_Elements/Deprecated/ProjectPointsOnLandscape.ProjectPointsOnLandscape_C) :BlueprintElementType /PCG/BP_Elements/Deprecated/ProjectPointsOnLandscape.ProjectPointsOnLandscape_C))
  (bind Transform%20Points_8 (Transform%20Points :bAbsoluteOffset true :OffsetMax (0.0 0.0 -32.0) :OffsetMin (0.0 0.0 -32.0)))
  (bind PointFilter_7 (Filter%20Attribute%20Elements :TargetAttribute $Position.z :ThresholdAttribute $Position.z))
  (bind StaticMeshSpawner_6 (Static%20Mesh%20Spawner :bApplyMeshBoundsToPoints false :MeshSelectorParameters (class:/Script/PCG.PCGMeshSelectorWeighted :MeshEntries [(struct :Descriptor (struct :StaticMesh /Game/Environments/Asteria/Foliage/ForestFloor/MossPatchClump.MossPatchClump :ComponentClass /Script/Engine.HierarchicalInstancedStaticMeshComponent :Mobility Movable :BodyInstance (struct :VelocitySolverIterationCount 1 :CollisionEnabled NoCollision :CollisionProfileName NoCollision :CollisionResponses (struct :ResponseArray [(struct :Channel Visibility :Response ECR_Ignore) (struct :Channel Camera :Response ECR_Ignore)]) :bAutoWeld true) :InstanceStartCullDistance 7000 :InstanceEndCullDistance 15000 :bGenerateOverlapEvents true :WorldPositionOffsetDisableDistance 500))])))
  (bind Distance_13 (Distance :bSetDensity true :MaximumDistance 716.476835))
  (bind DensityRemap_23 (Density%20Remap :OutRangeMax 0.0 :OutRangeMin 1.0))
  (bind Density%20Filter_21 (Density%20Filter :LowerBound 0.0))
  (bind Transform%20Points_24 (Transform%20Points :OffsetMax (0.0 0.0 -22.0) :OffsetMin (0.0 0.0 -1e+01) :ScaleMax (2.0 2.0 2.0)))
  (bind Self%20Pruning_25 (Self%20Pruning))
  (bind BoundsModifier_1 (Bounds%20Modifier :BoundsMax (-6e+02 -6e+02 1.0) :BoundsMin (6e+02 6e+02 1.0) :Mode Translate))
  (bind ExecuteBlueprint_2 (Execute%20Blueprint :BlueprintElementInstance (class:/PCG/BP_Elements/Deprecated/ScaleByDensity.ScaleByDensity_C :Density%20Exp 0.4 :Scale%20Max 1.0 :Scale%20Min 0.5) :BlueprintElementType /PCG/BP_Elements/Deprecated/ScaleByDensity.ScaleByDensity_C))
  ; old All in one node
  (bind ExecuteBlueprint_3 (Execute%20Blueprint :BlueprintElementInstance (class:/Game/ArtTools/Procedural/PCGMoss/FN_SampleMeshbyTag.FN_SampleMeshbyTag_C) :BlueprintElementType /Game/ArtTools/Procedural/PCGMoss/FN_SampleMeshbyTag.FN_SampleMeshbyTag_C))
  (connect ExecuteBlueprint_0 Out Density%20Filter_4 In)
  (connect DensityRemap_23 Out Density%20Filter_21 In)
  (connect DensityRemap_0 Out Density%20Filter_35 In)
  (connect Distance_33 Out DensityRemap_0 In)
  (connect Distance_13 Out DensityRemap_23 In)
  (connect Surface%20Sampler_1 Out Distance_13 Source)
  (connect ExecuteBlueprint_11 Out Distance_13 Target)
  (connect Density%20Filter_4 Out Distance_33 Source)
  (connect ExecuteBlueprint_11 Out Distance_33 Target)
  (connect ExecuteBlueprint_1 Out ExecuteBlueprint_0 In)
  (connect Get%20Actor%20Data_20 Out ExecuteBlueprint_0 In)
  (connect ExtentsModifier_46 Out ExecuteBlueprint_5 In)
  (connect DefaultInputNode Landscape ExecuteBlueprint_6 Landscape)
  (connect Self%20Pruning_45 Out ExecuteBlueprint_6 PointsToProject)
  (connect Transform%20Points_43 Out ExtentsModifier_46 In)
  (connect Transform%20Points_8 Out PointFilter_7 Filter)
  (connect Self%20Pruning_45 Out PointFilter_7 In)
  (connect Transform%20Points_24 Out Self%20Pruning_25 In)
  (connect ExecuteBlueprint_5 Out Self%20Pruning_45 In)
  (connect PointFilter_7 InsideFilter StaticMeshSpawner_6 In)
  (connect DefaultInputNode OriginalActor Surface%20Sampler_1 Bounding%20Shape)
  (connect DefaultInputNode Landscape Surface%20Sampler_1 Surface)
  (connect ExecuteBlueprint_6 Out Transform%20Points_8 In)
  (connect Density%20Filter_21 Out Transform%20Points_24 In)
  (connect Density%20Filter_35 Out Transform%20Points_43 In))
