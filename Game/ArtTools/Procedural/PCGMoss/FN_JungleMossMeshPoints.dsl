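; FN_JungleMossMeshPoints is a PCG graph in the procedural moss tools that outputs points for jungle moss, apparently sampled from tagged meshes.
; It wraps a single Execute Blueprint node running the custom FN_SampleMeshbyTagUnique element and sends that result straight to Out.
; The graph declares an untyped NewPin input, but nothing connects to it.
; All sampling logic lives in the Blueprint element, with no further filtering or transforms here.
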(graph FN_JungleMossMeshPoints
  (inputs (NewPin :type Any))
  (bind ExecuteBlueprint_0 (Execute%20Blueprint :BlueprintElementInstance (class:/Game/ArtTools/Procedural/PCGMoss/FN_SampleMeshbyTagUnique.FN_SampleMeshbyTagUnique_C) :BlueprintElementType /Game/ArtTools/Procedural/PCGMoss/FN_SampleMeshbyTagUnique.FN_SampleMeshbyTagUnique_C))
  (output Out ExecuteBlueprint_0))
