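; PCG Blueprint element that turns static meshes on actors tagged "Moss" into point data, apparently for procedural moss.
; It maps component world transforms to meshes, builds per-mesh local sample lists, then composes them into world-space samples.
; IterationLoopBody converts samples, or vertex arrays, into position-seeded points.
; Geometry sampling is removed over build errors: voxelize is empty, vertex arrays are never filled here, and each mesh gets one origin sample.
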(fn Execute
  (params (Input :type PCGData%20Collection%20Structure%20%28by%20ref%29))
  (bind _output (|CallParentFunction Input))
  (return _output))

(fn ExecuteWithContext
  (params (InContext :type PCGContext%20Structure%20%28by%20ref%29) (Input :type PCGData%20Collection%20Structure%20%28by%20ref%29))
  (bind _meshtransformmap (Variables|Default|GetMeshTransformMap))
  (LocalVariables|SetContext InContext)
  (Utilities|FlowControl|Sequence
    (:then_0
      (bind _outactors (Actor|GetAllActorswithTag "Moss"))
      (for _array_element _outactors
        (for _array_element_2 (Actor|GetComponentsByClass _array_element "/Script/Engine.StaticMeshComponent")
          (Utilities|IsValid (Variables|/Script/Engine.StaticMeshComponent|GetStaticMesh _array_element_2)
            (:Is%20Valid
              (Utilities|Map|Add (Variables|Default|GetMeshTransformMap) (Transformation|GetWorldTransform _array_element_2) (Variables|/Script/Engine.StaticMeshComponent|GetStaticMesh _array_element_2))
              (bind _returnvalue (Utilities|Array|AddUnique (Variables|Default|GetUniqueMeshList) (Variables|/Script/Engine.StaticMeshComponent|GetStaticMesh _array_element_2)))
              (Development|PrintString (Utilities|String|Append (Utilities|String|Append (Utilities|String|Append "Test_" (Utilities|String|ToString(Transform) (Transformation|GetWorldTransform _array_element_2))) (Utilities|GetDisplayName _array_element)) (Utilities|GetDisplayName (Variables|/Script/Engine.StaticMeshComponent|GetStaticMesh _array_element_2)))))))))
    (:then_1
      (for _array_element_3 (Variables|Default|GetUniqueMeshList)
        (Variables|Default|SetStaticMesh _array_element_3)
        (Utilities|FlowControl|Sequence
          (:then_0
            ;[1 0 0.1319] Geometry sampling removed causes build errors
            (if (Variables|Default|GetVoxelize)
              (Utilities|FlowControl|Sequence
                (:then_1
                  _))
              (else
                _))
            (if (not (Variables|Default|GetPointsFromVertices))
              (Utilities|FlowControl|Sequence
                (:then_0
                  (Utilities|Array|Add (Variables|Default|GetMeshSampleList) (Utilities|Struct|MakeFNMeshSampleListStruct (Variables|Default|GetStaticMesh) (Utilities|Array|MakeArray (Math|Transform|MakeTransform "0, 0, 0")))))))))))
    (:then_2
      (Development|PrintString (Utilities|String|Append "Test_" (Utilities|String|ToString(Integer) (Utilities|Array|Length (Variables|Default|GetMeshSampleList))))))
    (:then_3
      (for _array_element_4 (Variables|Default|GetMeshSampleList)
        (bind _values (Utilities|Map|Values _meshtransformmap))
        (bind _keys (Utilities|Map|Keys _meshtransformmap))
        (for (_array_element_5 _array_index) _values
          (if (== _array_element_5 (Utilities|Struct|BreakFNMeshSampleListStruct _array_element_4))
            (for _array_element_6 (.TransformSampleList_5_18D70A8041A9E814DE17E8BE0AC7D6BE (Utilities|Struct|BreakFNMeshSampleListStruct _array_element_4))
              (Utilities|Array|Add (Variables|Default|GetSamples) (Math|Transform|ComposeTransforms _array_element_6 (Utilities|Array|Get(acopy) _keys _array_index))))))))
    (:then_4
      (LocalVariables|SetIterations (Utilities|Array|Length (Variables|Default|GetSamples)))
      (bind _returnvalue_2 (Game|ConstructObjectfromClass "/Script/PCG.PCGPointData"))
      (bind _returnvalue_3 (PCG|Temporary|GetComponent (Variables|GetInContext)))
      (Variables|/Script/PCG.PCGSpatialData|SetTargetActor (Components|GetOwner _returnvalue_3) _returnvalue_2)
      (bind _output_get (LocalVariables|SetOutData _returnvalue_2)))
    (:then_5
      (return (Utilities|Struct|MakePCGDataCollection (Utilities|Array|MakeArray (PCG|TaggedData|MakeTaggedData (LocalVariables|GetOutData))))))))

(fn IterationLoopBody
  (params (InContext :type PCGContext%20Structure%20%28by%20ref%29) (Iteration :type Integer64) (InA :type PCGSpatial%20Data%20Object%20Reference) (InB :type PCGSpatial%20Data%20Object%20Reference) (OutMetadata :type PCGMetadata%20Object%20Reference))
  (bind _returnvalue (Math|Conversions|ToInteger(Integer64) Iteration))
  (bind (_r _g _b _a) (Math|Color|BreakColor (Utilities|Array|Get(aref) (Variables|Hidden|GetColors) _returnvalue)))
  (bind _samples (Variables|Default|GetSamples))
  (bind _output (Utilities|Array|Get(acopy) _samples _returnvalue))
  (bind _returnvalue_2 (PCG|Helpers|ComputeSeedfromPosition (.location _output)))
  (if (Variables|Default|GetPointsFromVertices)
    (return (Utilities|Struct|MakePCGPoint (Math|Transform|MakeTransform (Utilities|Array|Get(aref) (Variables|Hidden|GetPositions) _returnvalue) (Math|Rotator|MakeRotfromZ (Utilities|Array|Get(aref) (Variables|Hidden|GetNormals) _returnvalue))) (Math|Float|SelectFloat _r 1.0 (Variables|per-vertexoptions|GetUseRedAsDensity)) "-1.000000,-1.000000,-1.000000" "1.000000,1.000000,1.000000" (Math|Vector4|MakeVector4 _r _g _b _a) _returnvalue_2) true)
    (else
      (return (Utilities|Struct|MakePCGPoint _output 1.0 "-1.000000,-1.000000,-1.000000" "1.000000,1.000000,1.000000" _returnvalue_2) true))))
